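# Conversation database for the Ruckus NPC in Battle Pass Season 15, entered through Conversation.Entry.Ruckus.
# A root speech node links to the shared BasicServices and Season 15 quest conversations and offers two local branches.
# BOUNTY requires the PlayerBounty service, then offers a confirm that grants a player bounty or a back option, each with a voice line.
# The flavor dialogue branch has three replies converging on three shared questions, all ending in one closing speech.
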ConversationDatabase /BattlepassS15/Conversations/NPC_Conversation_Ruckus
entry_points [{tag=Conversation.Entry.Ruckus, destinations [#1]}]
nodes [
 ConversationEntryPointNode_7 {
  id=#1
  class=ConversationEntryPointNode
  outputs [#2]
  properties {entryTag {tagName=Conversation.Entry.Ruckus}}
 }
 Speech_C_4 {
  id=#2
  class=Speech_C
  outputs [#3, #4, #5, #6]
  properties {nodeName=""}
 }
 Speech_C_9 {
  id=#6
  class=Speech_C
  outputs [#7, #8]
  properties {softTaskIcon=/Game/Athena/HUD/Interaction/T_UI_NPC_Bounty}
  sub_nodes [
   FortConversationRequirement_HasService_1 {class=FortConversationRequirement_HasService, properties {serviceTag {tagName=Tandem.Service.PlayerBounty}}}
   FortConversationChoiceWithPreviewNode_1 {class=FortConversationChoiceWithPreviewNode, properties {defaultChoiceDisplayText=BOUNTY}}
  ]
 }
 GrantPlayerBounty_C_0 {
  id=#8
  class=GrantPlayerBounty_C
  outputs []
  sub_nodes [
   ConversationChoiceNode_5 {class=ConversationChoiceNode, properties {defaultChoiceDisplayText=CONFIRM}}
   SideEffect_Client_PlayVOEvent_C_0 {class=SideEffect_Client_PlayVOEvent_C, properties {voiceOverTag {tagName=Pawn.Athena.NPC.Sound.Voice.Quest.Excited}}}
  ]
 }
 Back_C_0 {
  id=#7
  class=Back_C
  sub_nodes [
   ConversationChoiceNode_8 {class=ConversationChoiceNode, properties {defaultChoiceDisplayText=BACK}}
   SideEffect_Client_PlayVOEvent_C_1 {class=SideEffect_Client_PlayVOEvent_C, properties {voiceOverTag {tagName=Pawn.Athena.NPC.Sound.Voice.Quest.Disappointed}}}
  ]
 }
 Speech_C_0 {
  id=#5
  class=Speech_C
  outputs [#9, #10, #11]
  properties {softTaskIcon=/Game/Athena/HUD/Interaction/T_UI_NPC_Interrobang}
  sub_nodes [ConversationChoiceNode_1 {class=ConversationChoiceNode, properties {defaultChoiceDisplayText=You... changed?}}]
 }
 Speech_C_3 {
  id=#11
  class=Speech_C
  outputs [#12, #13, #14]
  sub_nodes [ConversationChoiceNode_0 {class=ConversationChoiceNode, properties {defaultChoiceDisplayText=I wanted better loot.}}]
 }
 Speech_C_7 {
  id=#14
  class=Speech_C
  outputs [#15]
  properties {softTaskIcon=&T_UI_NPC_question}
  sub_nodes [ConversationChoiceNode_2 {class=ConversationChoiceNode, properties {defaultChoiceDisplayText=Is it permanent?}}]
 }
 Continue_C_0 {
  id=#15
  class=Continue_C
  outputs [#16]
  properties {softTaskIcon=/Game/Athena/HUD/Interaction/T_UI_NPC_statement}
 }
 Speech_C_8 {id=#16, class=Speech_C, outputs []}
 Speech_C_6 {
  id=#13
  class=Speech_C
  outputs [#15]
  properties {softTaskIcon=&T_UI_NPC_question}
  sub_nodes [ConversationChoiceNode_6 {class=ConversationChoiceNode, properties {defaultChoiceDisplayText=Can I do that too?}}]
 }
 Speech_C_5 {
  id=#12
  class=Speech_C
  outputs [#15]
  properties {softTaskIcon=&T_UI_NPC_question}
  sub_nodes [ConversationChoiceNode_7 {class=ConversationChoiceNode, properties {defaultChoiceDisplayText=Are you a ghost?}}]
 }
 Speech_C_2 {
  id=#10
  class=Speech_C
  outputs [#12, #13, #14]
  sub_nodes [ConversationChoiceNode_3 {class=ConversationChoiceNode, properties {defaultChoiceDisplayText="Uh, not my fault."}}]
 }
 Speech_C_1 {
  id=#9
  class=Speech_C
  outputs [#12, #13, #14]
  sub_nodes [ConversationChoiceNode_4 {class=ConversationChoiceNode, properties {defaultChoiceDisplayText=You started it.}}]
 }
 ConversationLinkNode_0 {id=#4, class=ConversationLinkNode, properties {remoteEntryTag {tagName=Conversation.Entry.BasicServices}}}
 ConversationLinkNode_1 {id=#3, class=ConversationLinkNode, properties {remoteEntryTag {tagName=Conversation.Entry.Season15.Quests}}}
]
