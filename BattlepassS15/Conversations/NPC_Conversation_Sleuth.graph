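# Conversation database for the Sleuth NPC in the Season 15 battle pass content, offering quests, services and a player bounty.
# The sleuth entry tag opens a root speech node that branches to shared Season 15 Quests and BasicServices conversations through link nodes, plus a local BOUNTY option.
# The bounty option requires the PlayerBounty service and shows a preview choice, then offers CONFIRM to grant it or BACK, each playing a client-side voice-over reaction.
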ConversationDatabase /BattlepassS15/Conversations/NPC_Conversation_Sleuth
entry_points [{tag=Conversation.Entry.sleuth, destinations [#1]}]
nodes [
 ConversationEntryPointNode_7 {
  id=#1
  class=ConversationEntryPointNode
  outputs [#2]
  properties {entryTag {tagName=Conversation.Entry.sleuth}}
 }
 Speech_C_4 {
  id=#2
  class=Speech_C
  outputs [#3, #4, #5]
  properties {nodeName=""}
 }
 ConversationLinkNode_0 {id=#5, class=ConversationLinkNode, properties {remoteEntryTag {tagName=Conversation.Entry.BasicServices}}}
 Speech_C_1 {
  id=#4
  class=Speech_C
  outputs [#6, #7]
  properties {softTaskIcon=/Game/Athena/HUD/Interaction/T_UI_NPC_Bounty}
  sub_nodes [
   FortConversationRequirement_HasService_6 {class=FortConversationRequirement_HasService, properties {serviceTag {tagName=Tandem.Service.PlayerBounty}}}
   FortConversationChoiceWithPreviewNode_6 {class=FortConversationChoiceWithPreviewNode, properties {defaultChoiceDisplayText=BOUNTY}}
  ]
 }
 GrantPlayerBounty_C_0 {
  id=#7
  class=GrantPlayerBounty_C
  outputs []
  sub_nodes [
   ConversationChoiceNode_1 {class=ConversationChoiceNode, properties {defaultChoiceDisplayText=CONFIRM}}
   SideEffect_Client_PlayVOEvent_C_0 {class=SideEffect_Client_PlayVOEvent_C, properties {voiceOverTag {tagName=Pawn.Athena.NPC.Sound.Voice.Quest.Excited}}}
  ]
 }
 Back_C_0 {
  id=#6
  class=Back_C
  sub_nodes [
   ConversationChoiceNode_6 {class=ConversationChoiceNode, properties {defaultChoiceDisplayText=BACK}}
   SideEffect_Client_PlayVOEvent_C_1 {class=SideEffect_Client_PlayVOEvent_C, properties {voiceOverTag {tagName=Pawn.Athena.NPC.Sound.Voice.Quest.Disappointed}}}
  ]
 }
 ConversationLinkNode_1 {id=#3, class=ConversationLinkNode, properties {remoteEntryTag {tagName=Conversation.Entry.Season15.Quests}}}
]
