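# Test NPC conversation database for Battlepass S15 that offers a Jenkins test quest through dialogue.
# The TestNpc1 entry point opens a speech node with two "QUEST!" preview choices and a link to the shared BasicServices conversation.
# Each quest choice leads to a speech node with a back option and an accept option.
# Both accept options grant the same static quest, and the choice labels appear to be placeholders.
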ConversationDatabase /BattlepassS15/Conversations/NPC_Conversation_Test1
entry_points [{tag=Conversation.Entry.TestNpc1, destinations [#1]}]
nodes [
 ConversationEntryPointNode_0 {
  id=#1
  class=ConversationEntryPointNode
  outputs [#2]
  properties {entryTag {tagName=Conversation.Entry.TestNpc1}}
 }
 Speech_C_4 {
  id=#2
  class=Speech_C
  outputs [#3, #4, #5]
  properties {nodeName=""}
 }
 ConversationLinkNode_0 {id=#5, class=ConversationLinkNode, properties {remoteEntryTag {tagName=Conversation.Entry.BasicServices}}}
 Speech_C_7 {
  id=#4
  class=Speech_C
  outputs [#6, #7]
  properties {softTaskIcon=/Game/UI/Foundation/Textures/Icons/HUD/T_PrimaryMinimap-Icon-64}
  sub_nodes [FortConversationChoiceWithPreviewNode_1 {class=FortConversationChoiceWithPreviewNode, properties {defaultChoiceDisplayText="QUEST!"}}]
 }
 GrantStaticQuest_C_1 {
  id=#7
  class=GrantStaticQuest_C
  outputs []
  properties {questToGrantPtr=/BattlepassS15/Items/QuestItems/JenkinsTest/quest_jenkinstest_ind_n01_q01}
  sub_nodes [ConversationChoiceNode_0 {class=ConversationChoiceNode, properties {defaultChoiceDisplayText="ACCEPT!"}}]
 }
 Back_C_3 {
  id=#6
  class=Back_C
  sub_nodes [ConversationChoiceNode_5 {class=ConversationChoiceNode, properties {defaultChoiceDisplayText="BACK!"}}]
 }
 Speech_C_8 {
  id=#3
  class=Speech_C
  outputs [#8, #9]
  properties {softTaskIcon=&T_PrimaryMinimap-Icon-64}
  sub_nodes [FortConversationChoiceWithPreviewNode_0 {class=FortConversationChoiceWithPreviewNode, properties {defaultChoiceDisplayText="QUEST!"}}]
 }
 GrantStaticQuest_C_0 {
  id=#9
  class=GrantStaticQuest_C
  outputs []
  properties {questToGrantPtr=&quest_jenkinstest_ind_n01_q01}
  sub_nodes [ConversationChoiceNode_1 {class=ConversationChoiceNode, properties {defaultChoiceDisplayText="ACCEPT!"}}]
 }
 Back_C_2 {
  id=#8
  class=Back_C
  sub_nodes [ConversationChoiceNode_17 {class=ConversationChoiceNode, properties {defaultChoiceDisplayText="BACK!"}}]
 }
]
