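# Shared Season 15 NPC conversation database that offers up to three slotted quests from the Conversation.Entry.Season15.Quests entry point.
# The entry fans out to three parallel Speech choices, each gated by its Tandem.Service.Quest tag, no active quests, completed slot prerequisites, and no completion this match.
# Each previewed choice leads to Accept, which grants that slot's quest with an affirmative client voice line, or Back, which plays a disappointed one.
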ConversationDatabase /BattlepassS15/Conversations/Share_Conversation_Quests
entry_points [{tag=Conversation.Entry.Season15.Quests, destinations [#1]}]
nodes [
 ConversationEntryPointNode_7 {
  id=#1
  class=ConversationEntryPointNode
  outputs [#2, #3, #4]
  properties {entryTag {tagName=Conversation.Entry.Season15.Quests}}
 }
 Speech_C_0 {
  id=#4
  class=Speech_C
  outputs [#5, #6]
  properties {softTaskIcon=/Game/Athena/HUD/Interaction/T_UI_NPC_quest.T_UI_NPC_Quest}
  sub_nodes [
   FortConversationRequirement_HasService_2 {class=FortConversationRequirement_HasService, properties {serviceTag {tagName=Tandem.Service.Quest3}}}
   FortConversationRequirement_HasNoActiveQuests_2 {class=FortConversationRequirement_HasNoActiveQuests}
   FortConversationRequirement_AllSlottedQuestPrerequisitesCompleted_2 {class=FortConversationRequirement_AllSlottedQuestPrerequisitesCompleted, properties {questSlot=3}}
   FortConversationChoiceWithPreviewNode_5 {class=FortConversationChoiceWithPreviewNode, properties {defaultChoiceDisplayText=Quest}}
   FortConversationRequirement_SlottedQuestNotCompletedThisMatch_2 {class=FortConversationRequirement_SlottedQuestNotCompletedThisMatch, properties {questSlot=3}}
  ]
 }
 Back_C_5 {
  id=#6
  class=Back_C
  sub_nodes [
   ConversationChoiceNode_8 {class=ConversationChoiceNode, properties {defaultChoiceDisplayText="BACK!"}}
   SideEffect_Client_PlayVOEvent_C_5 {class=SideEffect_Client_PlayVOEvent_C, properties {voiceOverTag {tagName=Pawn.Athena.NPC.Sound.Voice.Quest.Disappointed}}}
  ]
 }
 GrantSlottedQuest_C_2 {
  id=#5
  class=GrantSlottedQuest_C
  outputs []
  properties {questSlot=3}
  sub_nodes [
   ConversationChoiceNode_1 {class=ConversationChoiceNode, properties {defaultChoiceDisplayText="ACCEPT!"}}
   SideEffect_Client_PlayVOEvent_C_2 {class=SideEffect_Client_PlayVOEvent_C, properties {voiceOverTag {tagName=Pawn.Athena.NPC.Sound.Voice.Quest.Affirmative}}}
  ]
 }
 Speech_C_3 {
  id=#3
  class=Speech_C
  outputs [#7, #8]
  properties {softTaskIcon=&T_UI_NPC_Quest}
  sub_nodes [
   FortConversationRequirement_HasService_1 {class=FortConversationRequirement_HasService, properties {serviceTag {tagName=Tandem.Service.Quest2}}}
   FortConversationRequirement_HasNoActiveQuests_1 {class=FortConversationRequirement_HasNoActiveQuests}
   FortConversationRequirement_AllSlottedQuestPrerequisitesCompleted_1 {class=FortConversationRequirement_AllSlottedQuestPrerequisitesCompleted, properties {questSlot=2}}
   FortConversationChoiceWithPreviewNode_3 {class=FortConversationChoiceWithPreviewNode, properties {defaultChoiceDisplayText=Quest}}
   FortConversationRequirement_SlottedQuestNotCompletedThisMatch_1 {class=FortConversationRequirement_SlottedQuestNotCompletedThisMatch, properties {questSlot=2}}
  ]
 }
 Back_C_1 {
  id=#8
  class=Back_C
  sub_nodes [
   ConversationChoiceNode_7 {class=ConversationChoiceNode, properties {defaultChoiceDisplayText="BACK!"}}
   SideEffect_Client_PlayVOEvent_C_4 {class=SideEffect_Client_PlayVOEvent_C, properties {voiceOverTag {tagName=Pawn.Athena.NPC.Sound.Voice.Quest.Disappointed}}}
  ]
 }
 GrantSlottedQuest_C_1 {
  id=#7
  class=GrantSlottedQuest_C
  outputs []
  properties {questSlot=2}
  sub_nodes [
   ConversationChoiceNode_6 {class=ConversationChoiceNode, properties {defaultChoiceDisplayText="ACCEPT!"}}
   SideEffect_Client_PlayVOEvent_C_1 {class=SideEffect_Client_PlayVOEvent_C, properties {voiceOverTag {tagName=Pawn.Athena.NPC.Sound.Voice.Quest.Affirmative}}}
  ]
 }
 Speech_C_16 {
  id=#2
  class=Speech_C
  outputs [#9, #10]
  properties {softTaskIcon=&T_UI_NPC_Quest}
  sub_nodes [
   FortConversationRequirement_HasService_0 {class=FortConversationRequirement_HasService, properties {serviceTag {tagName=Tandem.Service.Quest1}}}
   FortConversationRequirement_HasNoActiveQuests_0 {class=FortConversationRequirement_HasNoActiveQuests}
   FortConversationRequirement_AllSlottedQuestPrerequisitesCompleted_0 {class=FortConversationRequirement_AllSlottedQuestPrerequisitesCompleted}
   FortConversationChoiceWithPreviewNode_4 {class=FortConversationChoiceWithPreviewNode, properties {defaultChoiceDisplayText=Quest}}
   FortConversationRequirement_SlottedQuestNotCompletedThisMatch_0 {class=FortConversationRequirement_SlottedQuestNotCompletedThisMatch}
  ]
 }
 Back_C_4 {
  id=#10
  class=Back_C
  sub_nodes [
   ConversationChoiceNode_2 {class=ConversationChoiceNode, properties {defaultChoiceDisplayText="BACK!"}}
   SideEffect_Client_PlayVOEvent_C_3 {class=SideEffect_Client_PlayVOEvent_C, properties {voiceOverTag {tagName=Pawn.Athena.NPC.Sound.Voice.Quest.Disappointed}}}
  ]
 }
 GrantSlottedQuest_C_0 {
  id=#9
  class=GrantSlottedQuest_C
  outputs []
  sub_nodes [
   ConversationChoiceNode_0 {class=ConversationChoiceNode, properties {defaultChoiceDisplayText="ACCEPT!"}}
   SideEffect_Client_PlayVOEvent_C_0 {class=SideEffect_Client_PlayVOEvent_C, properties {voiceOverTag {tagName=Pawn.Athena.NPC.Sound.Voice.Quest.Affirmative}}}
  ]
 }
]
