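# Conversation database for shared service conversations in the Season 15 Battle Pass content.
# It is an empty shell with no entry points and no nodes, so it defines no dialogue flow, choices or tasks.
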ConversationDatabase /BattlepassS15/Conversations/Share_Conversation_Services
entry_points []
nodes []
