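# Share_Test is a conversation database in BattlepassS15, apparently a test of branching dialogue.
# An entry point tagged Abilities.Activities.Demobomb leads to an opening Speech node, which branches to two terminal Speech nodes.
# Each branch carries a choice sub-node with placeholder labels "Pick A" and "Pick B".
# No speech content, requirements or side effects are set on the nodes shown.
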ConversationDatabase /BattlepassS15/Conversations/Share_Test
entry_points [{tag=Abilities.Activities.Demobomb, destinations [#1]}]
nodes [
 ConversationEntryPointNode_0 {
  id=#1
  class=ConversationEntryPointNode
  outputs [#2]
  properties {entryTag {tagName=Abilities.Activities.Demobomb}}
 }
 Speech_C_2 {id=#2, class=Speech_C, outputs [#3, #4]}
 Speech_C_5 {
  id=#4
  class=Speech_C
  outputs []
  sub_nodes [ConversationChoiceNode_1 {class=ConversationChoiceNode, properties {defaultChoiceDisplayText=Pick B}}]
 }
 Speech_C_3 {
  id=#3
  class=Speech_C
  outputs []
  sub_nodes [ConversationChoiceNode_6 {class=ConversationChoiceNode, properties {defaultChoiceDisplayText=Pick A}}]
 }
]
